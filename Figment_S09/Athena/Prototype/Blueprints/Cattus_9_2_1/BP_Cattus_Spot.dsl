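; Cattus event actor that surfaces an ice castle mesh offshore, moves it along a spline, then sinks it after a set duration.
; ZUp and ZDown play the WaterFlow timeline forward and reversed, driving height, pitch and particle radius, with timed audio and hiding.
; Non-authority Tick interpolates the looping spline position and moves mesh and water VFX.
; Replay pause freezes movement, scrubbing hides the actor, and a material parameter collection drives water-flow opacity.
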(event Custom|ReplayPaused
  (params (bPaused :type Boolean))
  (Variables|Default|SetIsReplayPaused bPaused))

(event Custom|CE_WaterFlow
  (params (FroceHidden :type Boolean))
  (bind _s_icecastle_lowres_fx (Variables|Default|GetS_IceCastle_LowRes_FX))
  ;[0 0 0] FX
  (if FroceHidden
    (Utilities|FlowControl|Sequence
      (:then_0
        (bind (_direction _opacity) (Timeline|Timeline_0 Stop)
          (:Update
            (Rendering|Material|SetScalarParameterValue "/Figment_S09/Athena/Prototype/Blueprints/Cattus_9_2_1/FX/MPC_CattusShader.MPC_CattusShader" "WaterFlowOpacity" _opacity :!class /Script/Engine.KismetMaterialLibrary)
            (Development|PrintString (Utilities|String|ToString(Float) _opacity) true true "(R=1.000000,G=0.463450,B=0.000000,A=1.000000)" 0.01))
          (:Finished
            _)))
      (:then_1
        _))
    (Development|SetHiddeninGame _s_icecastle_lowres_fx true false)
    (else
      (Development|SetHiddeninGame _s_icecastle_lowres_fx false false)
      (Rendering|Material|SetScalarParameterValue "/Figment_S09/Athena/Prototype/Blueprints/Cattus_9_2_1/FX/MPC_CattusShader.MPC_CattusShader" "WaterFlowOpacity" 1.5 :!class /Script/Engine.KismetMaterialLibrary)
      (bind _returnvalue (Utilities|Time|SetTimerbyEvent (delegate Custom|WaterFlowDelayed) 15.0))
      (Variables|Default|SetWaterFlorTimerHandle _returnvalue))))

(event Custom|WaterFlowDelayed
  ;[0 0 0] FX
  (Timeline|Timeline_0 PlayFromStart))

(event EventBeginPlay
  (bind _returnvalue (BlueprintContext|GetFortReplayContext))
  ;[1 1 1] Totally hidden in replays?
  (if (Replay|GetIsPlayingReplay)
    (Variables|Default|SetCanTrigger true)
    (ReplayContext|BindEventtoOnPauseStateChanged _returnvalue (EventDispatchers|CreateEvent :!event Custom|ReplayPaused))
    (ReplayContext|BindEventtoOnReplayScrubComplete _returnvalue (EventDispatchers|CreateEvent :!event Custom|ReplayScrubbed))
    (else
      (Variables|Default|SetCanTrigger true))))

(event Custom|SetVFXLocation
  (params (Radius :type Float%20%28double-precision%29) (CenterPos :type Vector))
  (bind _p_waterraid (Variables|Default|GetP_WaterRaid))
  ;[0 0 0] FX
  (Transformation|SetWorldLocation _p_waterraid CenterPos false false)
  (Effects|Components|ParticleSystem|SetFloatParameter _p_waterraid "BP_Radius" (+ (* Radius 1e+03) (Math|Random|RandomFloatinRange 2e+01 8e+01)))
  (Effects|Components|ParticleSystem|SetFloatParameter _p_waterraid "BP_RadiusVel" (/ (Math|Random|RandomFloatinRange 0.6 1.6) Radius))
  (Effects|Components|ParticleSystem|SetFloatParameter _p_waterraid "BP_Rate" (+ Radius 0.25)))

(event EventTick
  (params (DeltaSeconds :type Float%20%28single-precision%29))
  (if (not (Variables|Default|GetIsReplayPaused))
    (Networking|SwitchHasAuthority
      (:Remote
        (bind _output_get (Variables|Default|SetSmoothedLocationOnSpline (Math|Interpolation|FInterpTo (Variables|Default|GetSmoothedLocationOnSpline) (Variables|Default|GetNextLocationOnSpline) DeltaSeconds 3.0)))
        (Variables|Default|SetNextLocationOnSpline (+ (Variables|Default|GetNextLocationOnSpline) (* DeltaSeconds (Variables|Default|GetAmountToMove))))
        (if (>= _output_get 1.0)
          (Variables|Default|SetSmoothedLocationOnSpline (Math|Float|%25%28Float%29 (Variables|Default|GetSmoothedLocationOnSpline)))
          (Variables|Default|SetNextLocationOnSpline (Math|Float|%25%28Float%29 (Variables|Default|GetNextLocationOnSpline))))
        (CallFunction|SetPosition (Variables|Default|GetSmoothedLocationOnSpline))))))

(event Custom|SetStartLocation
  (params (StartLocationOnSpline :type Float%20%28double-precision%29))
  (Variables|Default|SetPosition StartLocationOnSpline)
  (Development|PrintString "Called here to set the location!"))

(event Custom|SetPosition
  (params (Time :type Float%20%28double-precision%29))
  (bind _spline (Variables|Default|GetSpline))
  (bind _returnvalue (Spline|GetTransformatTime _spline Time "World" true false))
  (bind _location (.location _returnvalue))
  (bind _rotation (.rotation _returnvalue))
  (bind _returnvalue_2 (Math|Transform|MakeTransform _location _rotation))
  (bind _rotationoffset (Variables|Default|GetRotationOffset))
  (bind _zoffset (Variables|Default|GetZOffset))
  (bind _returnvalue_3 (Math|Transform|MakeTransform _zoffset _rotationoffset))
  (bind _returnvalue_4 (Math|Transform|ComposeTransforms _returnvalue_3 _returnvalue_2))
  (if (<= (.z (Variables|Default|GetZOffset)) (+ (Variables|Default|GetHowHighOutOfWater) 1e+03))
    (Variables|Default|SetZOffset (+ (Variables|Default|GetZOffset) (Math|Vector|MakeVector 0.0 0.0 (Math|Random|RandomFloatinRange -1.0 1.0))))
    (else
      _))
  (Transformation|SetWorldTransform (Variables|Default|GetS_IceCastle_LowRes) _returnvalue_4 false false)
  (CallFunction|SetVFXLocation (Variables|Default|GetRadius) (Math|Vector|MakeVector (.x (.location _returnvalue_4)) (.y (.location _returnvalue_4)) -5e+03)))

(event Custom|SetParametersAndStartEvent
  (params (Position :type Float%20%28double-precision%29) (Duration :type Float%20%28double-precision%29) (ShoreDistance :type Float%20%28double-precision%29))
  (if (Variables|Default|GetCanTrigger)
    (Variables|Default|SetPosition Position)
    (Variables|Default|SetDuration Duration)
    (Variables|Default|SetShoreDistance ShoreDistance)
    (CallFunction|ZUp)
    (Development|PrintString "All params set, start event!!" true true "(R=0.000000,G=0.660000,B=1.000000,A=1.000000)" 6e+01)))

(event Custom|SetSpeed
  (params (Speed :type Float%20%28double-precision%29))
  (Variables|Default|SetAmountToMove (/ Speed 1e+06)))

(event Custom|ZUpFadeIn
  (Audio|Components|Audio|FadeIn (Variables|Default|GetAudio) 7.0 1.0 0.0 "Linear" :!class /Script/Engine.AudioComponent))

(event Custom|ZUp
  (bind _position (Variables|Default|GetPosition))
  (Networking|SwitchHasAuthority
    (:Remote
      (Utilities|FlowControl|Sequence
        (:then_0
          (Components|Timeline|Stop (Variables|BP_Cattus_Spot|GetWaterFlow)))
        (:then_1
          (Components|Timeline|Stop (Variables|BP_Cattus_Spot|GetTimeline_0)))
        (:then_2
          (Development|SetHiddeninGame (Variables|Default|GetS_IceCastle_LowRes_FX) false false))
        (:then_3
          (Utilities|Time|ClearandInvalidateTimerbyHandle (Variables|Default|GetCattusDeactivationTimerHandle))
          (Utilities|Time|ClearandInvalidateTimerbyHandle (Variables|Default|GetWaterFlorTimerHandle))
          (Utilities|Time|ClearandInvalidateTimerbyHandle (Variables|Default|GetZUpFadeInTimerHandle)))
        (:then_4
          (Utilities|Time|SetTimerbyEvent (EventDispatchers|CreateEvent :!event Custom|ZDown) (Variables|Default|GetDuration))
          (Variables|Default|SetSmoothedLocationOnSpline _position)
          (Variables|Default|SetNextLocationOnSpline (+ _position (Variables|Default|GetAmountToMove)))
          (Actor|Tick|SetActorTickEnabled true)
          (Audio|SpawnSoundAttached "/Figment_S09/Athena/Sounds/Events/Cattus/Cattus_Up_Cue.Cattus_Up_Cue" (Variables|Default|GetAudio))
          (Utilities|FlowControl|Sequence
            (:then_0
              (bind _returnvalue (Utilities|Time|SetTimerbyEvent (delegate Custom|ZUpFadeIn) 4.5))
              (Variables|Default|SetZUpFadeInTimerHandle _returnvalue))
            (:then_1
              (Rendering|SetActorHiddenInGame self false)
              (Utilities|FlowControl|Sequence
                (:then_0
                  (Components|Activation|Activate (Variables|Default|GetP_WaterRaid) true)
                  (Components|Timeline|SetPlayRate (Variables|BP_Cattus_Spot|GetWaterFlow) 0.1)
                  (bind (_direction _z_transform _radius) (Timeline|WaterFlow PlayFromStart)
                    (:Update
                      (Variables|Default|SetRadius _radius)
                      (Variables|Default|SetZOffset (Math|Vector|MakeVector 0.0 (Variables|Default|GetShoreDistance) (* _z_transform (Variables|Default|GetHowHighOutOfWater))))
                      (if (Utilities|Enum|Equal(Enum) _direction "Backward")
                        (Variables|Default|SetRotationOffset (Math|Rotator|MakeRotator (* (- _z_transform 1.0) (* (Variables|Default|GetPitchRot) -1.0)) 0.0 -9e+01))
                        (else
                          (Variables|Default|SetRotationOffset (Math|Rotator|MakeRotator (* (- _z_transform 1.0) (Variables|Default|GetPitchRot)) 0.0 -9e+01)))))
                    (:Finished
                      (if (Utilities|Enum|Equal(Enum) _direction "Backward")
                        (Actor|Tick|SetActorTickEnabled false)
                        (Components|Activation|Deactivate (Variables|Default|GetP_WaterRaid))
                        (bind _returnvalue_2 (Utilities|Time|SetTimerbyEvent (delegate Custom|HideCattus) 6.0))
                        (Variables|Default|SetCattusDeactivationTimerHandle _returnvalue_2)))))
                (:then_1
                  (CallFunction|CEWaterFlow false))))))))))

(event Custom|ZDown
  (Audio|Components|Audio|FadeOut (Variables|Default|GetAudio) 5.0 0.0 "Linear" :!class /Script/Engine.AudioComponent)
  (Audio|SpawnSoundAttached "/Figment_S09/Athena/Sounds/Events/Cattus/Cattus_Down_Cue.Cattus_Down_Cue" (Variables|Default|GetAudio))
  (Components|Timeline|SetPlayRate (Variables|BP_Cattus_Spot|GetWaterFlow) 0.2)
  (Timeline|WaterFlow ReverseFromEnd))

(event Custom|HideCattus
  (Rendering|SetActorHiddenInGame self true))

(event Custom|ReplayScrubbed
  (Actor|Tick|SetActorTickEnabled false)
  (Components|Activation|Deactivate (Variables|Default|GetP_WaterRaid))
  (Rendering|SetActorHiddenInGame self true))

(fn ConstructionScript
  (CallFunction|SetPosition (Variables|Default|GetPosition)))
